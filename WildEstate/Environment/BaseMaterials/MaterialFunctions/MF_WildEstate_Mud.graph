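# Landscape material function for WildEstate that blends a stylized mud layer into incoming Material Attributes and also outputs a mask, a StaticMesh material and array-texture size/index data.
# The layer weight, perturbed by world noise that fades under Gravel and Sand, is rounded into a hard mask, an outline and stepped transition bands.
# These masks drive two-tone mud color, with a darkening outline on the mobile shading path, multiplied by a world-noise normal-override result.
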MaterialFunction /WildEstate/Environment/BaseMaterials/MaterialFunctions/MF_WildEstate_Mud
graph {nodes [
  n1 = ScalarParameter() {defaultValue=800, group=Asphalt, parameterName=Mud_OverlayTexture_Size}
  Divide(A=WorldPosition().XY, B=@n1)
  n2 = LinearInterpolate(
   A=VectorParameter() {group=Dirt, parameterName=MudColor2, defaultValue=<rgba 0.3073 0.1748 0.0727 1>}.RGB
   B=VectorParameter() {group=Mud, parameterName=MudColor, defaultValue=<rgba 0.2188 0.1004 0.0296 1>}.RGB
   Alpha=NamedRerouteUsage()
  )
  n3 = LinearInterpolate(A=Constant3Vector(), B=@n2, Alpha=NamedRerouteUsage())
  n4 = SetMaterialAttributes!(
   MaterialAttributes
   BaseColor=Multiply(
    A=ShadingPathSwitch(Default=@n2, Mobile=@n3) {desc=outline for ES3}
    B=MaterialFunctionCall!(NormalOverrideWS) {materialFunction=&/MF_WildEstateWorldNoise}.Result
   )
   Roughness=Constant() {desc=CellShadingValue, r=0.9}
   Specular=Constant()
  ) {attributeSetTypes [#1, #2, #3]}
  FunctionOutput(None=BlendMaterialAttributes(A=FunctionInput() {inputName=In Material, inputType=FunctionInput_MaterialAttributes}, B=@n4, Alpha=NamedRerouteUsage())) {outputName=Out Material, sortPriority=10}
  n5 = AppendVector(A=@n1, B=@n1)
  n6 = Constant() {desc=ArrrayIndex ID, r=2}
  FunctionOutput(None=LinearInterpolate(
    A=FunctionInput() {bUsePreviewValueAsDefault=true, inputName=ArrayTextureUV, sortPriority=100}
    B=AppendVector(A=@n5, B=@n6)
    Alpha=NamedRerouteUsage()
   )) {outputName=ArrayTextureUV, sortPriority=100}
  FunctionOutput(None=NamedRerouteUsage()) {outputName=Mask, sortPriority=30}
  FunctionOutput(None=@n4) {outputName=StaticMesh, sortPriority=11}
  LandscapeLayerSample() {parameterName=Sand}
  n7 = Add(A=LandscapeLayerSample() {parameterName=Gravel}, B=LandscapeLayerSample() {parameterName=Sand})
  n8 = NamedRerouteDeclaration(None=FunctionInput(Preview=Constant()) {bUsePreviewValueAsDefault=true, inputName=LayerWeight, inputType=FunctionInput_Scalar, sortPriority=30}) {name=DirtMask_Original, nodeColor=<rgba 1 0 0.9844 1>}
  n9 = Add(
   A=LinearInterpolate(
    A=MaterialFunctionCall() {materialFunction=&MF_WildEstateWorldNoise}.TransitionNoise
    Alpha=Saturate(None=@n7)
   ) {constB=0}
   B=@n8
  )
  n10 = Subtract(A=@n8, B=ScalarParameter() {defaultValue=0.2, group=Mud, parameterName=MudThreshold})
  n11 = Add(A=Multiply(A=Saturate(None=@n9), B=ScalarParameter() {defaultValue=0.075, group=Mud, parameterName=SizeMudTransitionAdd}), B=@n10)
  NamedRerouteDeclaration(None=Saturate(None=Round(None=@n11))) {name=DirtMask, nodeColor=<rgba 1 0 0.9844 1>}
  n12 = Round(None=Add(A=@n10, B=ScalarParameter() {defaultValue=-0.0038, group=Dirt, parameterName=LayerTransition_LineWidth}))
  NamedRerouteDeclaration(None=Max(A=Saturate(None=@n12), B=ScalarParameter() {defaultValue=0.75, group=Dirt, parameterName=LayerTransition_LineDarkness})) {name=OutlineMask, nodeColor=<rgba 1 0.1172 0 1>}
  n13 = FunctionInput(Preview=Constant()) {bUsePreviewValueAsDefault=true, inputName=Additional LayerWeight for Transitions, inputType=FunctionInput_Scalar, sortPriority=50}
  n14 = Multiply(A=@n8, B=OneMinus(None=@n13))
  n15 = ScalarParameter() {defaultValue=-0.296, group=Mud, parameterName=MudTransitionBandingBias}
  n16 = ScalarParameter() {defaultValue=4, group=Mud, parameterName=MudTransitionBanding}
  n17 = Round(None=Multiply(A=Add(A=@n14, B=@n15), B=@n16))
  NamedRerouteDeclaration(None=Saturate(None=Divide(A=@n17, B=@n16))) {name=DirtMask_TransitionBanding, nodeColor=<rgba 1 0 0.9844 1>}
 ]}
