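# Material function that layers stylized, cel-shaded gravel scree over an input material for the WildEstate environment.
# Gravel normal and height textures are projected on world XY.
# The normal, dotted with the sunlight vector and rounded, gives two-tone shadowing baked into base color, with fixed roughness and specular.
# HeightLerp combines LayerWeight with gravel height, rounded to a hard-edged blend; the raw weight also passes out as Mask.
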MaterialFunction /WildEstate/Environment/BaseMaterials/MaterialFunctions/MF_WildEstate_Scree
graph {nodes [
  n1 = FunctionInput(Preview=Constant()) {bUsePreviewValueAsDefault=true, inputName=LayerWeight, inputType=FunctionInput_Scalar, sortPriority=30}
  FunctionOutput(None=@n1) {outputName=Mask, sortPriority=30}
  n2 = CollectionParameter() {
   collection=&FortniteMaterialParameters
   parameterName=SunlightDirectionalVector
  }
  n3 = Divide(A=WorldPosition().XY, B=ScalarParameter() {defaultValue=528.6447, group=Scree, parameterName=Scree_Size})
  n4 = TextureSample(UVs=@n3) {samplerType=SAMPLERTYPE_Normal, texture=/WildEstate/Landscape/Textures/WildEstate_Gravel_N}
  n5 = Round(None=DotProduct(A=ComponentMask(None=@n2) {b=true, g=true, r=true}, B=Transform(None=@n4.RGB)))
  n6 = Multiply(
   A=VectorParameter() {group=Scree, parameterName=ScreeColor, defaultValue=<rgba 0.151 0.1445 0.1298 1>}.RGB
   B=LinearInterpolate(A=ScalarParameter() {defaultValue=0.68, group=Scree, parameterName=Scree_MinShadow}, Alpha=Saturate(None=@n5))
  )
  n7 = Constant() {desc=CellShadingValue, r=0.75}
  n8 = Constant()
  n9 = MaterialFunctionCall(
   Transition Phase=@n1
   Height Texture=LinearInterpolate(
    A=ScalarParameter() {defaultValue=-0.6975, group=Scree, parameterName=Scree_min}
    B=ScalarParameter() {defaultValue=1, group=Scree, parameterName=Scree_max}
    Alpha=TextureSample(UVs=@n3) {samplerType=SAMPLERTYPE_Masks, texture=/WildEstate/Landscape/Textures/WildEstate_Gravel_height}.G
   )
   Contrast=ScalarParameter() {defaultValue=1277076992, group=Scree, parameterName=ScreeTransition_Contrast}
  ) {materialFunction=&HeightLerp}
  FunctionOutput(None=BlendMaterialAttributes(
    A=FunctionInput() {inputName=In Material, inputType=FunctionInput_MaterialAttributes}
    B=SetMaterialAttributes!(MaterialAttributes, BaseColor=@n6, Roughness=@n7, Specular=@n8) {attributeSetTypes [#1, #2, #3]}
    Alpha=Round(None=@n9.Lerp Alpha No Contrast)
   )) {outputName=Out Material, sortPriority=1}
 ]}
