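# Landscape material function that layers stylized snow over incoming material attributes, with two snow colors mixed by world noise and flat roughness and specular.
# The layer weight minus a threshold is rounded into a hard-edged blend mask, and the mobile path adds a black outline to the color.
# It also outputs base color, mask and blended texture-array UVs.
# A camera-facing, distance-faded flipbook snow-card billboard chain is present but reaches no output.
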MaterialFunction /WildEstate/Landscape/M_WildEstate_Snow
graph {nodes [
  Add()
  n1 = Divide(
   A=AppendVector(A=WorldPosition().XY, B=Constant())
   B=ScalarParameter() {defaultValue=720.6542, group=Billboard, parameterName=Billboard_CellSize}
  )
  n2 = Subtract(A=Frac(None=@n1)) {constB=0.5}
  AppendVector(A=@n2, B=Constant())
  n3 = ScalarParameter() {defaultValue=800, group=Asphalt, parameterName=Snow_OverlayTexture_Size}
  Divide(A=WorldPosition().XY, B=@n3)
  Floor(None=@n1)
  Frac(None=ComponentMask(None=@n1) {g=true, r=true})
  n4 = VertexNormalWS()
  n5 = LinearInterpolate(
   A=VectorParameter() {group=Snow, parameterName=SnowColor, defaultValue=<rgba 0.8 0.8139 0.8958 1>}.RGB
   B=VectorParameter() {group=Snow, parameterName=SnowColor_2, defaultValue=<rgba 0.2274 0.4397 0.6823 1>}.RGB
   Alpha=MaterialFunctionCall!(NormalOverrideWS=FunctionInput(Preview=@n4) {bUsePreviewValueAsDefault=true, inputName=NormalOverrideWS}) {materialFunction=/WildEstate/Environment/BaseMaterials/MaterialFunctions/MF_WildEstateWorldNoise}.Result
  ) {constA=1}
  n6 = ShadingPathSwitch(Default=@n5, Mobile=LinearInterpolate(A=Constant3Vector(), B=@n5, Alpha=NamedRerouteUsage())) {desc=outline for ES3}
  n7 = Constant() {desc=CellShadingValue, r=0.8}
  n8 = Constant()
  n9 = NamedRerouteDeclaration(None=FunctionInput(Preview=Constant()) {bUsePreviewValueAsDefault=true, inputName=LayerWeight, inputType=FunctionInput_Scalar, sortPriority=30}) {name=SnowMask_Original, nodeColor=<rgba 1 0 0.9844 1>}
  n10 = Subtract(A=@n9, B=ScalarParameter() {group=Snow, parameterName=SnowThreshold})
  n11 = NamedRerouteDeclaration(None=Saturate(None=Round(None=@n10))) {name=SnowMask, nodeColor=<rgba 1 0 0.9844 1>}
  FunctionOutput(None=BlendMaterialAttributes(
    A=FunctionInput() {inputName=In Material, inputType=FunctionInput_MaterialAttributes}
    B=SetMaterialAttributes!(MaterialAttributes, BaseColor=@n6, Roughness=@n7, Specular=@n8) {attributeSetTypes [#1, #2, #3]}
    Alpha=@n11
   )) {outputName=Out Material}
  FunctionOutput(None=@n5) {outputName=BaseColor, sortPriority=10}
  FunctionOutput(None=@n11) {outputName=Mask, sortPriority=3}
  n12 = AppendVector(A=@n3, B=@n3)
  n13 = Constant() {desc=ArrrayIndex ID, r=1}
  FunctionOutput(None=LinearInterpolate(
    A=FunctionInput() {bUsePreviewValueAsDefault=true, inputName=ArrayTextureUV, sortPriority=100}
    B=AppendVector(A=@n12, B=@n13)
    Alpha=NamedRerouteUsage()
   )) {outputName=ArrayTextureUV, sortPriority=100}
  LandscapeLayerSample() {parameterName=BiomeC}
  n14 = VectorNoise(World Position=@n1)
  n15 = Constant() {r=4}
  n16 = Transform(None=Constant3Vector() {constant=<rgba 1 0 0 0>}) {transformSourceType=TRANSFORMSOURCE_Camera, transformType=TRANSFORM_Tangent}
  n17 = Constant3Vector() {constant=<rgba 1 1 0 0>}
  n18 = DotProduct(A=@n2, B=Normalize(VectorInput=Multiply(A=@n16, B=@n17)))
  n19 = Multiply(
   A=Transform(None=Constant3Vector() {constant=<rgba 0 0 1 0>}) {transformSourceType=TRANSFORMSOURCE_Camera, transformType=TRANSFORM_Tangent}
   B=Constant3Vector() {constant=<rgba 1 1 0 0>}
  )
  n20 = Multiply(A=DotProduct(A=@n2, B=Normalize(VectorInput=@n19))) {constB=-1}
  n21 = ComponentMask(None=Add(A=AppendVector(A=@n18, B=@n20)) {constB=0.5}) {g=true, r=true}
  n22 = ComponentMask(None=MaterialFunctionCall(
    "Animation  Phase (0-1)"=ComponentMask(None=@n14) {g=true}
    Number of Rows=@n15
    Number of Columns=@n15
    Texture=TextureObject() {texture=/WildEstate/Environment/TestContent/WildEstate_SnowCard_Atlas}
    UVs=Saturate(None=@n21)
   ) {materialFunction=&FlipBook}.Result) {r=true}
  n23 = Subtract(
   A=Length(None=WorldPosition() {worldPositionShaderOffset=WPT_CameraRelativeNoOffsets}.XYZ) {is=class MaterialExpressionLength}
   B=ScalarParameter() {defaultValue=50192.5469, group=Billboard, parameterName=BillboardFadeBias}
  )
  n24 = ScalarParameter() {defaultValue=200, group=Billboard, parameterName=BillboardFadeRange}
  n25 = OneMinus(None=Saturate(None=Divide(A=@n23, B=@n24)))
  n26 = ComponentMask(None=@n14) {r=true}
  n27 = ScalarParameter() {defaultValue=0.8157, group=Billboard, parameterName=BillBoard_Threshold}
  LinearInterpolate(B=Saturate(None=Add(A=@n22, B=@n25)), Alpha=Ceil(None=Subtract(A=@n26, B=@n27))) {constA=1}
  n28 = Add(A=@n9, B=ScalarParameter() {defaultValue=-0.0038, parameterName=LayerTransition_LineWidth})
  NamedRerouteDeclaration(None=Saturate(None=Round(None=@n28))) {name=OutlineMask_Snow, nodeColor=<rgba 1 0.1172 0 1>}
  ScalarParameter() {defaultValue=200, parameterName=DepthRipples}
 ]}
