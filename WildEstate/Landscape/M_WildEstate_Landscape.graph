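# Nanite-enabled WildEstate landscape material that shades terrain, spawns grass and assigns physical materials.
# Chained layer functions stack grass, dirt, gravel, mud, asphalt, snow, cliffs, scree and detail, then feed color quantization, an ocean-floor switch and Curie landscape function.
# Layer weights, biome masks and Perlin noise drive five grass types, including snow, scree and bushes.
# Physical materials use escalating weights as priority; runtime virtual texture output and debug masks are included.
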Material /WildEstate/Landscape/M_WildEstate_Landscape
graph {
 properties {
  bAutomaticallySetUsageInEditor=false
  bTangentSpaceNormal=false
  bUseMaterialAttributes=true
  bUsedWithNanite=true
  usedShadingModels=Default Lit
 }
 nodes [
  n1 = LandscapeLayerSample() {parameterName=Sand}
  n2 = LandscapeLayerSample() {parameterName=Gravel}
  n3 = LandscapeLayerSample() {parameterName=Mud}
  n4 = Max(A=Max(A=Max(A=@n1, B=@n2), B=@n3), B=LandscapeLayerSample() {parameterName=Asphalt})
  Ceil(None=@n4)
  Constant()
  Constant()
  Constant()
  Constant() {r=1}
  Constant() {r=50}
  n5 = LandscapeVisibilityMask()
  n6 = LandscapeLayerSample() {parameterName=WildEstate_NoGrass}
  n7 = Round(None=Add(A=@n4, B=ScalarParameter() {defaultValue=0.49, parameterName=EdgeMargin}) {constB=0.4})
  n8 = OneMinus(None=NamedRerouteUsage())
  n9 = Saturate(None=Round(None=Add(A=@n8) {constB=0.01}))
  n10 = Subtract(A=Subtract(A=Constant() {r=1}, B=Saturate(None=@n7)), B=@n9)
  n11 = LandscapeLayerSample() {desc=Snow, parameterName=BiomeC}
  n12 = Saturate(None=Round(None=Add(A=@n11) {constB=0.4}))
  n13 = Multiply(A=@n5, B=Multiply(A=OneMinus(None=@n6), B=Subtract(A=@n10, B=@n12)))
  n14 = MaterialFunctionCall!(UseLandscapeCoords, "UseProjection(translucent)") {materialFunction=/WildEstate/Landscape/MaterialFunctions/MF_WildEstate_BiomeMasks}
  n15 = ComponentMask(None=@n14.QuadC_RGBA) {desc=MannicureGrass, r=true}
  MaterialExpressionMultiply_14 = Multiply(A=@n13, B=@n15)
  n16 = OneMinus(None=@n15)
  MaterialExpressionMultiply_13 = Multiply(A=@n13, B=@n16)
  n17 = Round(None=Add(A=LandscapeLayerSample() {desc=Snow, parameterName=BiomeC}) {constB=-0.4})
  MaterialExpressionMultiply_1 = Multiply(A=@n5, B=Saturate(None=Subtract(A=@n17, B=@n9)))
  MaterialExpressionMultiply_2 = Multiply(A=@n5, B=NamedRerouteUsage())
  n18 = TextureSample(UVs=Divide(A=WorldPosition().XY, B=ScalarParameter() {defaultValue=3000, parameterName=SmallBushMask_Size})) {samplerType=SAMPLERTYPE_Grayscale, texture=/WildEstate/Landscape/Textures/WildEstate_Landscape_Perlin_M}
  n19 = ScalarParameter() {defaultValue=0.2, parameterName=LowDensitySmallBushMask_threshold}
  n20 = Saturate(None=Divide(A=Subtract(A=@n18.R, B=@n19), B=ScalarParameter() {defaultValue=0.1, parameterName=LowDensitySmallBushMask_Falloff}))
  n21 = ScalarParameter() {defaultValue=0.1, parameterName=SmallBushMask_threshold}
  n22 = Saturate(None=Divide(A=Subtract(A=@n18.R, B=@n21), B=ScalarParameter() {defaultValue=0.1, parameterName=SmallBushMask_Falloff}))
  n23 = ComponentMask(None=@n14.QuadC_RGBA) {desc=DenseGrassandBushes, g=true}
  MaterialExpressionMultiply_11 = Multiply(A=Multiply(A=@n13, B=LinearInterpolate(A=@n20, B=@n22, Alpha=@n23)), B=@n16)
  LandscapeGrassOutput!(
   GrassManicured=@MaterialExpressionMultiply_14
   Grass=@MaterialExpressionMultiply_13
   Snow=@MaterialExpressionMultiply_1
   Scree=@MaterialExpressionMultiply_2
   SmallBushes=@MaterialExpressionMultiply_11
  ) {grassTypes [
    GrassManicured {
     grassType=/WildEstate/GrassTypes/WildEstate_GrassType_Manicured
     input {
      expression=&self:MaterialExpressionMultiply_14
      inputName=None
      mask=0
      maskA=0
      maskB=0
      maskG=0
      maskR=0
      outputIndex=0
     }
    }
    Grass {
     grassType=/WildEstate/GrassTypes/WildEstate_GrassType_Standard
     input {
      expression=&self:MaterialExpressionMultiply_13
      inputName=None
      mask=0
      maskA=0
      maskB=0
      maskG=0
      maskR=0
      outputIndex=0
     }
    }
    Snow {
     grassType=/WildEstate/GrassTypes/WildEstate_GrassType_Snow
     input {
      expression=&self:MaterialExpressionMultiply_1
      inputName=None
      mask=0
      maskA=0
      maskB=0
      maskG=0
      maskR=0
      outputIndex=0
     }
    }
    Scree {
     grassType=/WildEstate/GrassTypes/WildEstate_GrassType_Gravel
     input {
      expression=&self:MaterialExpressionMultiply_2
      inputName=None
      mask=0
      maskA=0
      maskB=0
      maskG=0
      maskR=0
      outputIndex=0
     }
    }
    SmallBushes {
     grassType=/WildEstate/GrassTypes/WildEstate_GrassType_SmallBushes
     input {
      expression=&self:MaterialExpressionMultiply_11
      inputName=None
      mask=0
      maskA=0
      maskB=0
      maskG=0
      maskR=0
      outputIndex=0
     }
    }
   ]}
  LandscapeLayerSample() {parameterName=Asphalt}
  LandscapeLayerSample() {parameterName=Grass}
  MaterialExpressionConstant_1 = Constant() {r=0.25}
  MaterialExpressionMultiply_7 = Multiply(A=LandscapeLayerSample() {parameterName=Sand}) {desc=Sand}
  MaterialExpressionMultiply_9 = Multiply(A=Max(A=LandscapeLayerSample() {parameterName=Gravel}, B=LandscapeLayerSample() {parameterName=Mud})) {constB=2, desc=Dirt}
  MaterialExpressionMultiply_10 = Multiply(A=LandscapeLayerSample() {parameterName=Asphalt}) {constB=4, desc=Asphalt}
  MaterialExpressionMultiply_5 = Multiply(A=OneMinus(None=NamedRerouteUsage())) {constB=5, desc=Stone}
  MaterialExpressionMultiply_6 = Multiply(A=LandscapeLayerSample() {parameterName=BiomeC}) {constB=6, desc=Snow}
  LandscapePhysicalMaterialOutput!(
   Grass=@MaterialExpressionConstant_1
   Sand=@MaterialExpressionMultiply_7
   Dirt=@MaterialExpressionMultiply_9
   Asphalt=@MaterialExpressionMultiply_10
   Stone=@MaterialExpressionMultiply_5
   Snow=@MaterialExpressionMultiply_6
  ) {
   desc="These are layered in order of importance.\r\nIf Dirt=1 and Stone=1, then Dirt will \"win\".\r\nThese values aren't clamped.\r\nSo a value of 2 should trump a value of 1."
   inputs [
    {
     physicalMaterial=/Game/FortressPhysicalMaterials/Grass
     input {
      expression=&self:MaterialExpressionConstant_1
      inputName=None
      mask=0
      maskA=0
      maskB=0
      maskG=0
      maskR=0
      outputIndex=0
     }
    }
    {
     physicalMaterial=/Game/FortressPhysicalMaterials/Sand
     input {
      expression=&self:MaterialExpressionMultiply_7
      inputName=None
      mask=0
      maskA=0
      maskB=0
      maskG=0
      maskR=0
      outputIndex=0
     }
    }
    {
     physicalMaterial=/Game/FortressPhysicalMaterials/Dirt
     input {
      expression=&self:MaterialExpressionMultiply_9
      inputName=None
      mask=0
      maskA=0
      maskB=0
      maskG=0
      maskR=0
      outputIndex=0
     }
    }
    {
     physicalMaterial=/Game/FortressPhysicalMaterials/Asphalt
     input {
      expression=&self:MaterialExpressionMultiply_10
      inputName=None
      mask=0
      maskA=0
      maskB=0
      maskG=0
      maskR=0
      outputIndex=0
     }
    }
    {
     physicalMaterial=/Game/FortressPhysicalMaterials/Stone
     input {
      expression=&self:MaterialExpressionMultiply_5
      inputName=None
      mask=0
      maskA=0
      maskB=0
      maskG=0
      maskR=0
      outputIndex=0
     }
    }
    {
     physicalMaterial=/Game/FortressPhysicalMaterials/Snow
     input {
      expression=&self:MaterialExpressionMultiply_6
      inputName=None
      mask=0
      maskA=0
      maskB=0
      maskG=0
      maskR=0
      outputIndex=0
     }
    }
   ]
  }
  n24 = MaterialFunctionCall!(NormalOverrideWS, UseCamera facing BildBoards, UseLandscapeCoords) {materialFunction=/WildEstate/Environment/BaseMaterials/MaterialFunctions/MF_WildEstate_Grass}
  n25 = LandscapeLayerSample() {parameterName=Asphalt}
  n26 = LandscapeLayerSample() {parameterName=Gravel}
  n27 = MaterialFunctionCall!(
   In Material=@n24.Out Material
   LayerWeight=LandscapeLayerSample() {parameterName=Sand}
   Additional LayerWeight for Transitions=Add(A=Add(A=@n25, B=@n26), B=LandscapeLayerSample() {parameterName=Mud})
   ArrayTextureUV=@n24.Out Material
  ) {materialFunction=/WildEstate/Environment/BaseMaterials/MaterialFunctions/MF_WildEstate_Dirt}
  n28 = MaterialFunctionCall!(
   In Material=@n27.Out Material
   LayerWeight=LandscapeLayerSample() {parameterName=Gravel}
   Additional LayerWeight for Transitions=Add(A=LandscapeLayerSample() {parameterName=Asphalt}, B=LandscapeLayerSample() {parameterName=Mud})
   ArrayTextureUV=@n27.Out Material
  ) {materialFunction=/WildEstate/Environment/BaseMaterials/MaterialFunctions/MF_WildEstate_Gravel2}
  n29 = MaterialFunctionCall!(
   In Material=@n28.Out Material
   LayerWeight=LandscapeLayerSample() {parameterName=Mud}
   Additional LayerWeight for Transitions=LandscapeLayerSample() {parameterName=Asphalt}
   ArrayTextureUV=@n28.Out Material
  ) {materialFunction=/WildEstate/Environment/BaseMaterials/MaterialFunctions/MF_WildEstate_Mud}
  n30 = MaterialFunctionCall!(In Material=@n29.Out Material, LayerWeight=LandscapeLayerSample() {parameterName=Asphalt}, ArrayTextureUV=@n29.Out Material) {materialFunction=/WildEstate/Environment/BaseMaterials/MaterialFunctions/MF_WildEstate_Asphalt}
  n31 = MaterialFunctionCall!(In Material=@n30.Out Material, NormalOverrideWS, LayerWeight=LandscapeLayerSample() {parameterName=BiomeC}, ArrayTextureUV=@n30.Out Material) {materialFunction=&/M_WildEstate_Snow}
  n32 = MaterialFunctionCall!(In Material=@n31.Out Material, UseLandscapeCoords) {materialFunction=/WildEstate/Environment/BaseMaterials/MaterialFunctions/MF_WildEstate_Cliffs}
  n33 = MaterialFunctionCall!(In Material=@n32.Out Material, LayerWeight=LandscapeLayerSample() {parameterName=BiomeB}) {materialFunction=/WildEstate/Environment/BaseMaterials/MaterialFunctions/MF_WildEstate_Scree}
  n34 = MaterialFunctionCall!(
   In Material=@n33.Out Material
   ArrayTextureUV=@n31.ArrayTextureUV
   CliffMask=NamedRerouteDeclaration(None=@n32.Mask) {name=CliffMask_mask, nodeColor=<rgba 0.0547 0 1 1>}
  ) {materialFunction=/WildEstate/Environment/BaseMaterials/MaterialFunctions/MF_WildEstate_LandscapeDetailTexture}
  n35 = GetMaterialAttributes(None=@n34.Out Material) {attributeGetTypes [#1]}
  n36 = Multiply(
   A=@n35.Base Color
   B=MaterialFunctionCall() {materialFunction=/WildEstate/Environment/BaseMaterials/MaterialFunctions/MF_WildEstate_LandscapeCloseUpTexture}.Out Material
  )
  LinearInterpolate(A=@n36, B=@n35.Base Color, Alpha=NamedRerouteUsage())
  n37 = Constant()
  n38 = StaticSwitchParameter(
   True=MaterialFunctionCall() {materialFunction=&MF_WildEstate_Dirt}.StaticMesh
   False=MaterialFunctionCall!(In=SetMaterialAttributes!(MaterialAttributes=@n35.MaterialAttributes, Base Color=@n36, Metallic=@n37, Specular=@n37) {attributeSetTypes [#1, #2, #3]}) {materialFunction=/WildEstate/Environment/BaseMaterials/MaterialFunctions/M_WildEstate_QantizeColor}.Result
  ) {parameterName=IsOceanFloor}
  n39 = SetMaterialAttributes!(
   MaterialAttributes=MaterialFunctionCall!(
    In=NamedRerouteDeclaration(None=@n38) {name=Final Material, nodeColor=<rgba 0 0.0859 1 1>}
    Enabled
    Disabled
    useDirectionalShading
    DirectionalShadingMask
   ) {materialFunction=/WildEstate/Environment/BaseMaterials/MaterialFunctions/MF_Test5050Unlit}.Result
   Opacity Mask=LandscapeVisibilityMask()
  ) {attributeSetTypes [#4]}
  n40 = GetMaterialAttributes(None=@n39)
  n41 = Multiply(A=LandscapeLayerSample() {parameterName=WildEstate_NoGrass}, B=Constant3Vector() {constant=<rgba 1 0 0 0>})
  MaterialFunctionCall!(
   In=StaticSwitchParameter(
    True=SetMaterialAttributes!(MaterialAttributes=@n40.MaterialAttributes, Emissive Color=@n41) {attributeSetTypes [#5]}
    False=@n39
   ) {group=Debug, parameterName=DebugNoGrassLayer}
   UseHeightFalloff
   OnFireEmissiveOverride
   HeightFalloffRadius
   HeightFalloffHardness
   VariationMasksMobile
   In_Grass
   Is_Stylised_WE=StaticBool() {value=true}
  ) {materialFunction=/Game/Packages/Fortress_SharedMaterials/Base_Material_Functions/MF_Curie_Landscape}
  MaterialFunctionCall(Float3=@n14.QuadC_RGBA) {materialFunction=&BreakOutFloat3Components}
  MaterialFunctionCall(Float3=@n14.QuadA_RGBA) {materialFunction=&BreakOutFloat3Components}
  Min(A=Min(A=@n1, B=@n2), B=@n3)
  NamedRerouteDeclaration(None=@n13) {name=Debug Grass layer, nodeColor=<rgba 1 0.3047 0 1>}
  n42 = RuntimeVirtualTextureSampleParameter(World Position=WorldPosition().XYZ) {
   bEnableFeedback=false
   materialType=BaseColor_Normal_Specular_YCoCg
   parameterName=RVT_Base
   virtualTexture=/Game/Athena/Apollo/Maps/Landscape/VT/RVT_Apollo_Landscape
  }
  NamedRerouteDeclaration(None=@n42.BaseColor) {name=RVT_BaseColor, nodeColor=<rgba 0 0.7656 1 1>}
  n43 = RuntimeVirtualTextureSampleParameter(World Position=WorldPosition().XYZ) {
   bEnableFeedback=false
   materialType=WorldHeight
   parameterName=RVT_WorldHeight
   virtualTexture=/Game/Athena/Apollo/Maps/Landscape/VT/RVT_Apollo_Landscape_Height
  }
  NamedRerouteDeclaration(None=@n43.WorldHeight) {name=RVT_WordPosZ, nodeColor=<rgba 0 0.7656 1 1>}
  NamedRerouteDeclaration(None=@n42.Normal) {name=RVT_Normal_WS, nodeColor=<rgba 0 0.7656 1 1>}
  n44 = GetMaterialAttributes(None=NamedRerouteUsage()) {attributeGetTypes [#1]}
  n45 = Multiply(A=@n44.Base Color, B=VectorParameter() {group=Debug, parameterName=DEBUG-TraversalMaskColor, defaultValue=<rgba 1 0 0 1>}.RGB)
  n46 = Constant() {r=43}
  n47 = Constant() {r=-16.5}
  n48 = WorldPosition()
  n49 = CrossProduct(A=DDX(Value=@n48.XYZ), B=DDY(Value=@n48.XYZ))
  n50 = Abs(None=ComponentMask(None=Normalize(VectorInput=@n49)) {b=true})
  n51 = Constant() {r=1}
  n52 = If(A=Saturate(None=LinearInterpolate(A=@n46, B=@n47, Alpha=@n50)), B=Constant() {r=0.75}, "A > B"=@n51, "A == B"=@n51, "A < B"=Constant())
  n53 = Multiply(A=ScalarParameter() {defaultValue=0.75, desc=Set this to 1 to turn it on, group=Debug, parameterName=DEBUG-TraversalMask}, B=@n52)
  NamedRerouteDeclaration(None=SetMaterialAttributes!(MaterialAttributes=@n44.MaterialAttributes, Base Color=LinearInterpolate(A=@n44.Base Color, B=@n45, Alpha=@n53)) {attributeSetTypes [#1]}) {name=DebugTraversalMask, nodeColor=<rgba 0.4062 1 0 1>}
  NamedRerouteDeclaration(None=@n33.Mask) {name=Scree_Mask, nodeColor=<rgba 1 0 0.8437 1>}
  NamedRerouteDeclaration(None=@n42.Roughness) {name=RVT_Roughness, nodeColor=<rgba 0 0.7656 1 1>}
  NamedRerouteDeclaration(None=@n30.Mask) {name=AsphaltMask, nodeColor=<rgba 0.3359 0 1 1>}
  NamedRerouteDeclaration(None=Constant() {r=1}) {name=TraversalMask, nodeColor=<rgba 0.6172 0 1 1>}
  NamedRerouteDeclaration(None=@n32.Mask_For_GrassSpawner) {name=CliffMask_Original, nodeColor=<rgba 0.0547 0 1 1>}
  NamedRerouteUsage() {desc=Don't use in shipping material}
  Reroute(None=OneMinus(None=@n52)) {desc=Dont use this in Grass mask}
  n54 = GetMaterialAttributes(None=@n34.Out Material) {attributeGetTypes [#1, #6, #3]}
  RuntimeVirtualTextureOutput!(
   BaseColor=@n54.Base Color
   Specular=@n54.Base Color
   Roughness=@n54.Base Color
   Normal=VertexNormalWS()
   WorldHeight=WorldPosition().Z
   Opacity
   Mask
   Displacement
   Mask4
  )
  SetMaterialAttributes!(MaterialAttributes=@n39, Emissive Color=NamedRerouteUsage()) {attributeSetTypes [#5]}
  StaticBool() {value=true}
  VirtualTextureFeatureSwitch!(
   No=@n34.Out Material
   Yes=SetMaterialAttributes!(MaterialAttributes, Base Color=NamedRerouteUsage(), Roughness=NamedRerouteUsage(), Normal=NamedRerouteUsage()) {attributeSetTypes [#1, #6, #7]}
  )
 ]
}
